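# Physics asset for the hood of the F_MED_RoseDust female player cosmetic, setting up physics bodies for its dynamic hoodie bones.
# A box body on spine_05 covers the upper torso and anchors a chain of two capsule bodies, dyn_hoodie_01 then dyn_hoodie_02.
# Both constraints lock twist and keep one swing axis narrow and the other wide, so the hood mainly flaps along a single axis.
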PhysicsAsset /BRCosmetics/Characters/Player/Female/Medium/Bodies/F_MED_RoseDust/Meshes/F_MED_RoseDust_Hood_Physics
bodies {
 dyn_hoodie_01 [{shapeType=Capsule, length=13.927, radius=3.3261, shapeName=dyn_hoodie_01_capsule, center=<2.3438 0.4642 3.5209>, rotation=<rot 0 0 -90>}]
 dyn_hoodie_02 [{
   shapeType=Capsule
   length=11.7619
   radius=3.1008
   shapeName=dyn_hoodie_02_capsule
   center=<1.7091 0.0549 1.3566>
   rotation=<rot 0 179.6494 -89.3743>
  }]
 spine_05 [{shapeType=Box, extentX=32.5, extentY=11, extentZ=27.5, shapeName=spine_05_box, center=<-0.5658 -2.6402 0>, rotation=<rot 0 9.9999 0>}]
}
constraints [
 {bone1Name=dyn_hoodie_02, bone2Name=dyn_hoodie_01, swing1LimitDegrees=10, swing2LimitDegrees=45, twistLimitDegrees=5, twistMotion=Locked}
 {bone1Name=spine_05, bone2Name=dyn_hoodie_01, swing1LimitDegrees=5, swing2LimitDegrees=90, twistLimitDegrees=5, twistMotion=Locked}
]
