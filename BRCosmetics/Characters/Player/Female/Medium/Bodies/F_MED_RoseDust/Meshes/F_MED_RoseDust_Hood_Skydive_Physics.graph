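# Physics asset for the skydive hood of the female medium RoseDust body, presumably providing secondary hood motion during skydiving.
# Two identically shaped capsules on dyn_hoodie_01 and dyn_hoodie_02 form a chain from a box body on spine_05 at the upper torso.
# Constraints link each bone to its parent with small angle limit values, and twist is locked between the two hood segments.
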PhysicsAsset /BRCosmetics/Characters/Player/Female/Medium/Bodies/F_MED_RoseDust/Meshes/F_MED_RoseDust_Hood_Skydive_Physics
bodies {
 dyn_hoodie_01 [{
   shapeType=Capsule
   length=11.7619
   radius=3.1008
   shapeName=dyn_hoodie_02_capsule
   center=<1.7091 0.0549 1.3566>
   rotation=<rot 0 179.6494 -89.3743>
  }]
 dyn_hoodie_02 [{
   shapeType=Capsule
   length=11.7619
   radius=3.1008
   shapeName=dyn_hoodie_02_capsule
   center=<1.7091 0.0549 1.3566>
   rotation=<rot 0 179.6494 -89.3743>
  }]
 spine_05 [{shapeType=Box, extentX=32.5, extentY=11, extentZ=27.5, shapeName=spine_05_box, center=<-0.5658 -2.6402 0>, rotation=<rot 0 9.9999 0>}]
}
constraints [
 {bone1Name=dyn_hoodie_02, bone2Name=dyn_hoodie_01, swing1LimitDegrees=10, swing2LimitDegrees=15, twistLimitDegrees=5, twistMotion=Locked}
 {bone1Name=dyn_hoodie_01, bone2Name=spine_05, swing1LimitDegrees=10, swing2LimitDegrees=15, twistLimitDegrees=5}
]
