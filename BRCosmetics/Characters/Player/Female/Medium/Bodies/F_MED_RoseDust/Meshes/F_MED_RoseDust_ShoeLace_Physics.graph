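# Physics asset for the dynamic shoelaces on both shoes of the female medium RoseDust character body.
# Capsule bodies cover calves, feet and lace bones; calf bodies have no constraints, likely serving as colliders.
# On each side, a master lace body is angularly locked to the foot and carries the knot, which branches into bow and inner-lace chains.
# Narrow swing and twist limits keep lace motion small.
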PhysicsAsset /BRCosmetics/Characters/Player/Female/Medium/Bodies/F_MED_RoseDust/Meshes/F_MED_RoseDust_ShoeLace_Physics
bodies {
 calf_l [{
   shapeType=Capsule
   length=13.25
   radius=3.25
   shapeName=calf_l_capsule
   center=<-37.1805 -0.5845 -0.6424>
   rotation=<rot 90 81.8176 -88.1812>
  }]
 calf_r [{shapeType=Capsule, length=13.25, radius=3.25, shapeName=calf_r_capsule, center=<37.0245 0.6125 0>, rotation=<rot 90 81.8176 91.818>}]
 dyn_laces_MSTR_l [{shapeType=Capsule, length=0.505, radius=0.505, shapeName=dyn_laces_MSTR_l_capsule, rotation=<rot 0 -90 0>}]
 dyn_laces_MSTR_r [{shapeType=Capsule, length=0.505, radius=0.505, shapeName=dyn_laces_MSTR_r_capsule, center=<0 -0.2864 0>, rotation=<rot 0 -90 0>}]
 dyn_laces_bow_inner_l_01 [{
   shapeType=Capsule
   length=0.6781
   radius=0.5775
   shapeName=dyn_laces_bow_inner_l_01_capsule
   center=<0.05 0.0007 0.0369>
   rotation=<rot 0 81.2934 -77.7573>
  }]
 dyn_laces_bow_inner_r_01 [{
   shapeType=Capsule
   length=1.2023
   radius=0.6325
   shapeName=dyn_laces_bow_inner_r_01_capsule
   center=<0.1263 -0.0123 0.018>
   rotation=<rot 0 96.0699 -80.811>
  }]
 dyn_laces_inner_l_01 [{
   shapeType=Capsule
   length=0.505
   radius=0.505
   shapeName=dyn_laces_inner_l_01_capsule
   center=<0.2606 -0.1715 0.0361>
   rotation=<rot -29.418 76.4728 -77.601>
  }]
 dyn_laces_inner_l_04 [{
   shapeType=Capsule
   length=3.75
   radius=1.005
   shapeName=dyn_laces_inner_l_04_capsule
   center=<-2.6275 -0.8122 0.3199>
   rotation=<rot 19.9976 -89.6018 90.9239>
  }]
 dyn_laces_inner_r_01 [{
   shapeType=Capsule
   length=1.0316
   radius=0.3702
   shapeName=dyn_laces_inner_r_01_capsule
   center=<-0.152 0.42 0.1522>
   rotation=<rot 8.4278 107.0745 -80.6898>
  }]
 dyn_laces_inner_r_04 [{
   shapeType=Capsule
   length=3.75
   radius=1.005
   shapeName=dyn_laces_inner_r_04_capsule
   center=<-2.429 1.071 0.1806>
   rotation=<rot -8.6544 -90 -89.4251>
  }]
 dyn_laces_knot_l [{
   shapeType=Capsule
   length=1.2004
   radius=0.691
   shapeName=dyn_laces_knot_l_capsule
   center=<0.0485 -0.2602 -0.0935>
   rotation=<rot 0 88.09 -77.8012>
  }]
 dyn_laces_knot_r [{
   shapeType=Capsule
   length=1.0501
   radius=0.6599
   shapeName=dyn_laces_knot_r_capsule
   center=<-0.1428 0.0342 -0.0716>
   rotation=<rot 0 103.6167 -76.8497>
  }]
 foot_l [
  {shapeType=Capsule, length=14, radius=2.5, shapeName=foot_l_capsule, center=<-3.6802 6.6722 0.2214>, rotation=<rot 0 19.9995 90>}
  {shapeType=Capsule, length=14.5, radius=3.25, shapeName=foot_l_capsule, center=<-5.5661 6.5285 0>, rotation=<rot 0 -0.0002 90>}
 ]
 foot_r [
  {shapeType=Capsule, length=14.25, radius=2.25, shapeName=foot_r_capsule, center=<2.2784 -6.5731 -0.5623>, rotation=<rot 0 9.9996 90>}
  {shapeType=Capsule, length=14.25, radius=3.5, shapeName=foot_r_capsule, center=<5.6712 -6.5916 0>, rotation=<rot 0 -0.0002 90>}
 ]
}
constraints [
 {bone1Name=dyn_laces_bow_inner_l_01, bone2Name=dyn_laces_knot_l, swing1LimitDegrees=20, swing2LimitDegrees=20, twistLimitDegrees=10}
 {bone1Name=dyn_laces_inner_l_01, bone2Name=dyn_laces_knot_l, swing1LimitDegrees=20, swing2LimitDegrees=20, twistLimitDegrees=10}
 {bone1Name=dyn_laces_bow_inner_r_01, bone2Name=dyn_laces_knot_r, swing1LimitDegrees=20, swing2LimitDegrees=20, twistLimitDegrees=10}
 {bone1Name=dyn_laces_inner_r_01, bone2Name=dyn_laces_knot_r, swing1LimitDegrees=20, swing2LimitDegrees=20, twistLimitDegrees=10}
 {bone1Name=dyn_laces_knot_l, bone2Name=dyn_laces_MSTR_l, swing1LimitDegrees=20, swing2LimitDegrees=20, twistLimitDegrees=5}
 {
  bone1Name=dyn_laces_MSTR_l
  bone2Name=foot_l
  swing1LimitDegrees=45
  swing1Motion=Locked
  swing2LimitDegrees=45
  swing2Motion=Locked
  twistLimitDegrees=45
  twistMotion=Locked
 }
 {bone1Name=dyn_laces_knot_r, bone2Name=dyn_laces_MSTR_r, swing1LimitDegrees=20, swing2LimitDegrees=20, twistLimitDegrees=5}
 {
  bone1Name=dyn_laces_MSTR_r
  bone2Name=foot_r
  swing1LimitDegrees=45
  swing1Motion=Locked
  swing2LimitDegrees=45
  swing2Motion=Locked
  twistLimitDegrees=45
  twistMotion=Locked
 }
 {bone1Name=dyn_laces_inner_l_04, bone2Name=dyn_laces_inner_l_01, swing1LimitDegrees=20, swing2LimitDegrees=20, twistLimitDegrees=5}
 {bone1Name=dyn_laces_inner_r_04, bone2Name=dyn_laces_inner_r_01, swing1LimitDegrees=20, swing2LimitDegrees=20, twistLimitDegrees=5}
]
