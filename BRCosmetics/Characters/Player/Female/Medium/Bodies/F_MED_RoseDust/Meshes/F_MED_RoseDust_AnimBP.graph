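# Post-pose AnimBlueprint for the F_MED_RoseDust body cosmetic, adding deformation and secondary motion to shoelaces and hood.
# It copies the parent mesh pose through a deform Control Rig, constrains lace knots to leg bones, then simulates laces with RigidBody.
# Two hood RigidBody nodes crossfade on a skydiving bool (parachute, dive or float), sharing a gravity override from a deprecated helper, before a small hood bone offset.
# The event graph also sets normal/skydive alphas, which the AnimGraph never reads.
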AnimBlueprint /BRCosmetics/Characters/Player/Female/Medium/Bodies/F_MED_RoseDust/Meshes/F_MED_RoseDust_AnimBP
target_skeleton=&/F_MED_RoseDust_Skeleton
graphs [
 AnimGraph {nodes [
   Blend Poses by bool() {blendType=Linear}
   n1 = Control Rig(Source=Copy Pose From Mesh() {bUseAttachedParent=true}) {
    alphaBoolBlend {blendOption=Linear}
    controlRigAssetReference {blueprintRigClass=&/F_MED_RoseDust_DeformRig.F_MED_RoseDust_DeformRig_C}
   }
   n2 = "Constraint - dyn_laces_knot_l (PPPPP)"!(
    ComponentPose=Local To Component(LocalPose=@n1)
    Alpha=1
    ConstraintWeights_0=0.0001
    ConstraintWeights_1:float=0.3019
    ConstraintWeights_2:float=0.6917
    ConstraintWeights_3:float=0.0061
    ConstraintWeights_4:float=0.0002
   ) -> (Pose) {
    alphaBoolBlend {blendOption=Linear}
    boneToModify {boneName=dyn_laces_knot_l}
    constraintSetup [
     {targetBone {boneName=calf_l}}
     {targetBone {boneName=calf_twist_01_l}}
     {targetBone {boneName=foot_l}}
     {targetBone {boneName=calf_twist_02_l}}
     {targetBone {boneName=ball_l}}
    ]
    constraintWeights [1,1,1,1,1]
   }
   n3 = "Constraint - dyn_laces_knot_r (PPPPP)"!(
    ComponentPose=@n2
    Alpha=1
    ConstraintWeights_0=0.0001
    ConstraintWeights_1:float=0.3019
    ConstraintWeights_2:float=0.6917
    ConstraintWeights_3:float=0.0061
    ConstraintWeights_4:float=0.0002
   ) -> (Pose) {
    alphaBoolBlend {blendOption=Linear}
    boneToModify {boneName=dyn_laces_knot_r}
    constraintSetup [
     {targetBone {boneName=calf_r}}
     {targetBone {boneName=calf_twist_01_r}}
     {targetBone {boneName=foot_r}}
     {targetBone {boneName=calf_twist_02_r}}
     {targetBone {boneName=ball_r}}
    ]
    constraintWeights [1,1,1,1,1]
   }
   n4 = RigidBody(ComponentPose=@n3, Alpha=0.1) {
    bClampLinearTranslationLimitToRefPose=true
    bOverrideWorldGravity=true
    overridePhysicsAsset=&/F_MED_RoseDust_ShoeLace_Physics
    simulationSpace=BaseBoneSpace
    alphaBoolBlend {blendOption=Linear}
    baseBoneRef {boneName=root}
    componentLinearAccScale=<0.2 0.2 0.2>
    overrideWorldGravity {z=-980}
    simSpaceSettings {maxLinearAcceleration=1002, maxLinearVelocity=100}
   }
   n5 = Get skydiving() -> (skydiving:bool)
   n6 = Get GravityOverride()
   n7 = RigidBody!(ComponentPose=@n4, bAlphaBoolEnabled:bool=NOT Boolean(A=@n5), OverrideWorldGravity:Vector=@n6, ExternalForce) -> (Pose) {
    alphaInputType=Bool
    bClampLinearTranslationLimitToRefPose=true
    bOverrideWorldGravity=true
    overridePhysicsAsset=&/F_MED_RoseDust_Hood_Physics
    simulationSpace=BaseBoneSpace
    alphaBoolBlend {blendInTime=0.25, blendOption=Linear, blendOutTime=0.25}
    alphaScaleBiasClamp {bInterpResult=true}
    baseBoneRef {boneName=spine_05}
    componentLinearAccScale=<0.5 0.5 0.5>
    overrideWorldGravity {z=-980}
    simSpaceSettings {maxAngularVelocity=1000, maxLinearVelocity=1000}
   }
   n8 = RigidBody(ComponentPose=@n7, bAlphaBoolEnabled=@n5, OverrideWorldGravity=@n6) {
    alphaInputType=Bool
    bClampLinearTranslationLimitToRefPose=true
    bOverrideWorldGravity=true
    overridePhysicsAsset=&/F_MED_RoseDust_Hood_Skydive_Physics
    simulationSpace=BaseBoneSpace
    alphaBoolBlend {blendInTime=0.25, blendOption=Linear, blendOutTime=0.25}
    alphaScaleBiasClamp {bInterpResult=true, scale=0.4}
    baseBoneRef {boneName=spine_05}
    componentAppliedLinearAccClamp=<2000 2000 2000>
    componentLinearAccScale=<0.2 0.2 0.3>
    componentLinearVelScale=<0.2 0.2 0.3>
    overrideWorldGravity {z=980}
    simSpaceSettings {maxAngularVelocity=1000, maxLinearVelocity=1000}
   }
   n9 = "Transform (Modify) Bone - Bone: dyn_hoodie_01"(ComponentPose=@n8, Alpha=1, Translation="0,-1,0", Rotation="0,0,0", Scale="1,1,1") {translationMode=BMM_Additive, alphaBoolBlend {blendOption=Linear}, boneToModify {boneName=dyn_hoodie_01}, translation {y=-1}}
   Output Pose(Result=Component To Local(ComponentPose=@n9)) {layerGroup=None}
  ]}
 EventGraph {nodes [
   Event BlueprintUpdateAnimation()
   > n1 = Sequence!(execute) -> (then_0, then_1, then_2:Exec)
   > n2 = Branch(Condition=Get bIsSkydiveFloating())
   > Set skydive_alpha(skydive_alpha:double=0.5) -> (Output_Get:double)
   > Set normal_alpha(normal_alpha:double=0) -> (Output_Get:double)
   Set skydive_alpha(execute=@n2.else, skydive_alpha=0)
   > Set normal_alpha(normal_alpha=1)
   Set skydiving(
    execute=@n1.then_1
    skydiving:bool=OR Boolean!(A=Get bIsParachuteOpened(), B=Get bIsSkydiveDiveMode(), C:bool=Get bIsSkydiveFloating()) -> (ReturnValue)
   ) -> (Output_Get:bool)
   n3 = Gravity Override State Function DEPRECATED(
    execute=@n1.then_2
    CharacterPartAnimInstance=Self-Reference()
    GameTimeInSeconds=GetGameTimeInSeconds()
    GravityOverrideParams=Get Gravity Override Params()
   )
   > Set GravityOverride(GravityOverride=@n3)
  ]}
]
variables [
 Gravity Override Params {type=GravityOverrideParamsStruct}
 GravityOverride {type=Vector}
 normal_alpha {type=number}
 skydive_alpha {type=number}
 skydiving {type=boolean}
]
