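# Physics asset for the RoseDust female medium character's FaceAcc mesh part, providing capsule bodies and constraints for dynamic hair.
# Five hair chains (front left and right, back left, two left-side) hang from the head, with neck, upper spine and left clavicle capsules as unconstrained body volumes.
# Angular limits keep motion small, tightest on the twist-locked front-right chain.
# The last links of the front-right and side-back chains skip their immediate predecessor, constraining to an earlier bone.
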PhysicsAsset /BRCosmetics/Characters/Player/Female/Medium/Bodies/F_MED_RoseDust/Meshes/Parts/F_MED_RoseDust_FaceAcc_Physics
bodies {
 clavicle_l [{shapeType=Capsule, length=6, radius=5, shapeName=clavicle_l_capsule, center=<6.1033 -2.2118 0.2857>, rotation=<rot -70 179.9997 -179.9998>}]
 dyn_hair_bck_l_01 [{shapeType=Capsule, length=0.75, radius=1.5, shapeName=dyn_hair_bck_l_01_capsule, rotation=<rot 90 -180 -180>}]
 dyn_hair_bck_l_02 [{
   shapeType=Capsule
   length=3.25
   radius=2.75
   shapeName=dyn_hair_bck_l_02_capsule
   center=<1.5106 -2.55 4.1504>
   rotation=<rot 12.2302 173.4405 38.1734>
  }]
 dyn_hair_bck_l_03 [{
   shapeType=Capsule
   length=3.5
   radius=2
   shapeName=dyn_hair_bck_l_03_capsule
   center=<-1.1918 -2.5124 3.4636>
   rotation=<rot -27.7521 2.9965 128.85>
  }]
 dyn_hair_fr_l_01 [{shapeType=Capsule, length=0.25, radius=2, shapeName=dyn_hair_fr_l_01_capsule, center=<0 -1.7543 0>, rotation=<rot 90 -180 -180>}]
 dyn_hair_fr_l_02 [{shapeType=Capsule, length=0.75, radius=1.5, shapeName=dyn_hair_fr_l_02_capsule, center=<0 -1.7543 0>, rotation=<rot 90 -180 -180>}]
 dyn_hair_fr_r_01 [{shapeType=Capsule, length=0.75, radius=1.5, shapeName=dyn_hair_fr_r_01_capsule, center=<-0.2952 -0.811 0>, rotation=<rot 90 0 20.0003>}]
 dyn_hair_fr_r_02 [{shapeType=Capsule, length=0.75, radius=1.5, shapeName=dyn_hair_fr_r_02_capsule, center=<0 -2.3666 0>, rotation=<rot 90 90 90.0002>}]
 dyn_hair_fr_r_03 [{shapeType=Capsule, length=0.75, radius=1.5, shapeName=dyn_hair_fr_r_03_capsule, center=<0 -2.1102 0>, rotation=<rot 90 90 90.0002>}]
 dyn_hair_side_bck_l_01 [{shapeType=Capsule, length=0.75, radius=1.5, shapeName=dyn_hair_side_bck_l_01_capsule, rotation=<rot 90 -169.1541 -169.1544>}]
 dyn_hair_side_bck_l_02 [{shapeType=Capsule, length=0.75, radius=1.5, shapeName=dyn_hair_side_bck_l_02_capsule, rotation=<rot 90 77.2888 77.2894>}]
 dyn_hair_side_bck_l_03 [{shapeType=Capsule, length=2, radius=1.5, shapeName=dyn_hair_side_bck_l_03_capsule, rotation=<rot 90 77.2888 77.2894>}]
 dyn_hair_side_bck_l_04 [{shapeType=Capsule, length=2, radius=1.5, shapeName=dyn_hair_side_bck_l_04_capsule, rotation=<rot 90 77.2888 77.2894>}]
 dyn_hair_side_fr_l_01 [{shapeType=Capsule, length=0.75, radius=1.5, shapeName=dyn_hair_side_fr_l_01_capsule, center=<0 0 -1.2128>, rotation=<rot 90 180 -179.9995>}]
 dyn_hair_side_fr_l_02 [{
   shapeType=Capsule
   length=0.75
   radius=1.5
   shapeName=dyn_hair_side_fr_l_02_capsule
   center=<0 0 -0.3766>
   rotation=<rot 90 179.9999 -179.9994>
  }]
 dyn_hair_side_fr_l_03 [{
   shapeType=Capsule
   length=0.75
   radius=1.5
   shapeName=dyn_hair_side_fr_l_03_capsule
   center=<1.4689 0 1.2866>
   rotation=<rot 70 0.0006 -0.0001>
  }]
 dyn_hair_side_fr_l_04 [{shapeType=Capsule, length=0.75, radius=1.5, shapeName=dyn_hair_side_fr_l_04_capsule, center=<0 0 1.7114>, rotation=<rot 90 180 179.9994>}]
 head [{shapeType=Capsule, length=5.75, radius=9, shapeName=head_capsule, center=<5.7417 0 0.5185>, rotation=<rot 90 -172.4477 -172.4474>}]
 neck_02 [{shapeType=Capsule, length=8, radius=6.5, shapeName=neck_02_capsule, center=<0.0052 0.0294 0>, rotation=<rot 90 -169.3803 -159.3801>}]
 spine_05 [{shapeType=Capsule, length=14.75, radius=9.5, shapeName=spine_05_capsule, center=<4.0825 -4.8418 0>, rotation=<rot 0 -0.0001 0>}]
}
constraints [
 {bone1Name=dyn_hair_bck_l_02, bone2Name=dyn_hair_bck_l_01, swing1LimitDegrees=25, swing2LimitDegrees=25, twistLimitDegrees=10}
 {bone1Name=dyn_hair_bck_l_03, bone2Name=dyn_hair_bck_l_02, swing1LimitDegrees=25, swing2LimitDegrees=25, twistLimitDegrees=10}
 {bone1Name=dyn_hair_fr_l_02, bone2Name=dyn_hair_fr_l_01, swing1LimitDegrees=25, swing2LimitDegrees=25, twistLimitDegrees=10}
 {bone1Name=dyn_hair_fr_r_02, bone2Name=dyn_hair_fr_r_01, swing1LimitDegrees=5, swing2LimitDegrees=5, twistLimitDegrees=5, twistMotion=Locked}
 {bone1Name=dyn_hair_fr_r_03, bone2Name=dyn_hair_fr_r_01, swing1LimitDegrees=5, swing2LimitDegrees=5, twistLimitDegrees=5, twistMotion=Locked}
 {bone1Name=dyn_hair_side_bck_l_02, bone2Name=dyn_hair_side_bck_l_01, swing1LimitDegrees=10, swing2LimitDegrees=10, twistLimitDegrees=10}
 {bone1Name=dyn_hair_side_bck_l_03, bone2Name=dyn_hair_side_bck_l_02, swing1LimitDegrees=10, swing2LimitDegrees=10, twistLimitDegrees=10}
 {bone1Name=dyn_hair_side_bck_l_04, bone2Name=dyn_hair_side_bck_l_02, swing1LimitDegrees=25, swing2LimitDegrees=25, twistLimitDegrees=10}
 {bone1Name=dyn_hair_side_fr_l_02, bone2Name=dyn_hair_side_fr_l_01, swing1LimitDegrees=25, swing2LimitDegrees=25, twistLimitDegrees=10}
 {bone1Name=dyn_hair_side_fr_l_03, bone2Name=dyn_hair_side_fr_l_02, swing1LimitDegrees=25, swing2LimitDegrees=25, twistLimitDegrees=10}
 {bone1Name=dyn_hair_side_fr_l_04, bone2Name=dyn_hair_side_fr_l_03, swing1LimitDegrees=25, swing2LimitDegrees=25, twistLimitDegrees=10}
 {bone1Name=dyn_hair_fr_l_01, bone2Name=head, swing1LimitDegrees=25, swing2LimitDegrees=25, twistLimitDegrees=10}
 {bone1Name=dyn_hair_side_fr_l_01, bone2Name=head, swing1LimitDegrees=25, swing2LimitDegrees=25, twistLimitDegrees=10}
 {bone1Name=dyn_hair_side_bck_l_01, bone2Name=head, swing1LimitDegrees=25, swing2LimitDegrees=25, twistLimitDegrees=10}
 {bone1Name=dyn_hair_bck_l_01, bone2Name=head, swing1LimitDegrees=25, swing2LimitDegrees=25, twistLimitDegrees=10}
 {bone1Name=dyn_hair_fr_r_01, bone2Name=head, swing1LimitDegrees=5, swing2LimitDegrees=5, twistLimitDegrees=5, twistMotion=Locked}
]
