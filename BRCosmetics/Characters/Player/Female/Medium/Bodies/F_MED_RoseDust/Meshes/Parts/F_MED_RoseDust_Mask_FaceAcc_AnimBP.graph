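# Post-process-style cosmetic AnimBP for the RoseDust mask face accessory, animating its left and right hood bones.
# AnimGraph copies the attached parent's pose, constrains both bones to head and clavicle_r, adds spring-based AnimDynamics, then applies corrective bone offsets and scales.
# EventGraph maps clavicle, upper-arm and head rotations relative to spine_05 into alpha variables, varying with emote, targeting, front-end and backbling state.
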AnimBlueprint /BRCosmetics/Characters/Player/Female/Medium/Bodies/F_MED_RoseDust/Meshes/Parts/F_MED_RoseDust_Mask_FaceAcc_AnimBP
target_skeleton=&/F_MED_RoseDust_Mask_FaceAcc_Skeleton
graphs [
 AnimGraph {nodes [
   "Transform (Modify) Bone - Bone: dyn_hoodUp_r"(Alpha=1, Translation="-2,0,2", Rotation="0,0,0", Scale="1,1,1") {
    rotationMode=BMM_Additive
    translationMode=BMM_Additive
    alphaBoolBlend {blendOption=Linear}
    boneToModify {boneName=dyn_hoodUp_r}
    translation {x=-2, z=2}
   }
   "Transform (Modify) Bone - Bone: dyn_hoodUp_l"(Alpha=1, Translation="8.060,2.000,4.000", Rotation="-30.000,0.000,-0.000", Scale="1,1,1") {
    rotationMode=BMM_Additive
    translationMode=BMM_Additive
    alphaBoolBlend {blendOption=Linear}
    boneToModify {boneName=dyn_hoodUp_l}
    rotation {pitch=-30}
    translation=<8.06 2 4>
   }
   n1 = "Constraint - dyn_hoodUp_r (PP)"!(
    ComponentPose=Local To Component(LocalPose=Copy Pose From Mesh() {bUseAttachedParent=true})
    Alpha=0.9
    ConstraintWeights_0=0.9
    ConstraintWeights_1:float=0.1
   ) -> (Pose) {
    alphaBoolBlend {blendOption=Linear}
    boneToModify {boneName=dyn_hoodUp_r}
    constraintSetup [{targetBone {boneName=head}}, {targetBone {boneName=clavicle_r}}]
    constraintWeights [1,1]
   }
   n2 = "Constraint - dyn_hoodUp_l (PP)"!(ComponentPose=@n1, Alpha=0.3, ConstraintWeights_0=0.5, ConstraintWeights_1:float=0.5) -> (Pose) {
    alphaBoolBlend {blendOption=Linear}
    boneToModify {boneName=dyn_hoodUp_l}
    constraintSetup [{targetBone {boneName=head}}, {targetBone {boneName=clavicle_r}}]
    constraintWeights [1,1]
   }
   n3 = Anim Dynamics - Bone: dyn_hoodUp_l(ComponentPose=@n2, Alpha=0.7, ExternalForce="0,0,0") {
    bGravityOverrideInSimSpace=true
    bLinearSpring=true
    bOverrideLinearDamping=true
    bUseGravityOverride=true
    linearDampingOverride=0.9
    linearSpringConstant=100
    alphaBoolBlend {blendOption=Linear}
    boundBone {boneName=dyn_hoodUp_l}
    componentAppliedLinearAccClamp=<5000 5000 5000>
    componentLinearAccScale=<0.3 0.3 0.7>
    componentLinearVelScale=<0.3 0.3 0.3>
    gravityOverride {z=-980}
    physicsBodyDefinitions [{
      boundBone {boneName=dyn_hoodUp_l}
      boxExtents=<10 10 10>
      constraintSetup {angularLimitsMax=<0.5 0.5 45>, angularLimitsMin=<-0.5 -0.5 0>, angularTarget=<0 0 0>, linearAxesMax=<0 0 10>}
      localJointOffset=<0 0 0>
     }]
    relativeSpaceBone {boneName=spine_05}
   }
   n4 = Anim Dynamics - Bone: dyn_hoodUp_r(ComponentPose=@n3, Alpha=0.7, ExternalForce="0,0,0") {
    bGravityOverrideInSimSpace=true
    bLinearSpring=true
    bOverrideLinearDamping=true
    bUseGravityOverride=true
    linearDampingOverride=0.9
    linearSpringConstant=100
    alphaBoolBlend {blendOption=Linear}
    boundBone {boneName=dyn_hoodUp_r}
    componentAppliedLinearAccClamp=<5000 5000 5000>
    componentLinearAccScale=<0.3 0.3 0.7>
    componentLinearVelScale=<0.3 0.3 0.3>
    gravityOverride {z=-980}
    physicsBodyDefinitions [{
      boundBone {boneName=dyn_hoodUp_r}
      boxExtents=<10 10 10>
      constraintSetup {angularLimitsMax=<0.5 0.5 0>, angularLimitsMin=<-0.5 -0.5 45>, angularTarget=<0 0 0>, linearAxesMin=<0 0 -10>}
      localJointOffset=<0 0 0>
     }]
    relativeSpaceBone {boneName=spine_05}
   }
   n5 = "Transform (Modify) Bone - Bone: dyn_hoodUp_r"(ComponentPose=@n4, Alpha=1, Translation="1.402,0.000,6.000", Rotation="70.000,0.000,0.000", Scale="0.7,0.8,1") {
    rotationMode=BMM_Additive
    scaleMode=BMM_Replace
    translationMode=BMM_Additive
    alphaBoolBlend {blendOption=Linear}
    boneToModify {boneName=dyn_hoodUp_r}
    rotation {pitch=70}
    scale=<0.7 0.8>
    translation {x=1.402, z=6}
   }
   n6 = "Transform (Modify) Bone - Bone: dyn_hoodUp_l"(ComponentPose=@n5, Alpha=1, Translation="0,0,3", Rotation="0,0,0", Scale="0.9,1,1") {
    scaleMode=BMM_Replace
    translationMode=BMM_Additive
    alphaBoolBlend {blendOption=Linear}
    boneToModify {boneName=dyn_hoodUp_l}
    scale {x=0.9}
    translation {z=3}
   }
   n7 = "Transform (Modify) Bone - Bone: dyn_hoodUp_r"(ComponentPose=@n6, Alpha=1, Translation="2.316,-3.058,3.5", Rotation="0,0,0", Scale="1,1,1") {
    rotationMode=BMM_Additive
    translationMode=BMM_Additive
    alphaBoolBlend {blendOption=Linear}
    boneToModify {boneName=dyn_hoodUp_r}
    translation=<2.316 -3.058 3.5>
   }
   n8 = "Transform (Modify) Bone - Bone: dyn_hoodUp_l"(ComponentPose=@n7, Alpha=1, Translation="1.231,5,3", Rotation="-19.683,-3.616,10.628", Scale="0.4,0.5,1") {
    rotationMode=BMM_Additive
    scaleMode=BMM_Replace
    translationMode=BMM_Additive
    alphaBoolBlend {blendOption=Linear}
    boneToModify {boneName=dyn_hoodUp_l}
    rotation=<rot -19.683 -3.616 10.628>
    scale=<0.4 0.5>
    translation=<1.231 5 3>
   }
   n9 = "Transform (Modify) Bone - Bone: dyn_hoodUp_l"(ComponentPose=@n8, Alpha=1, Translation="0,0,6", Rotation="0,0,0", Scale="0.9,1,1") {
    scaleMode=BMM_Replace
    translationMode=BMM_Additive
    translationSpace=BCS_BoneSpace
    alphaBoolBlend {blendOption=Linear}
    alphaScaleBiasClamp {bInterpResult=true, interpSpeedDecreasing=5, interpSpeedIncreasing=5}
    boneToModify {boneName=dyn_hoodUp_l}
    scale {x=0.9}
    translation {z=6}
   }
   n10 = "Transform (Modify) Bone - Bone: dyn_hoodUp_r"(ComponentPose=@n9, Alpha=1, Translation="0,0,-10", Rotation="0,0,0", Scale="0.9,1,1") {
    scaleMode=BMM_Replace
    translationMode=BMM_Additive
    translationSpace=BCS_BoneSpace
    alphaBoolBlend {blendOption=Linear}
    alphaScaleBiasClamp {bInterpResult=true, interpSpeedDecreasing=5, interpSpeedIncreasing=5}
    boneToModify {boneName=dyn_hoodUp_r}
    scale {x=0.9}
    translation {z=-10}
   }
   Output Pose(Result=Component To Local(ComponentPose=@n10)) {layerGroup=None}
  ]}
 EventGraph {nodes [
   Event BlueprintUpdateAnimation()
   > n1 = Sequence()
   > Branch(Condition=NOT Boolean(A=Get bIsPlayingEmote()))
   n2 = GetOwnerSkeletalMeshComponent()
   n3 = BreakTransform(InTransform=GetDeltaTransformFromRefPose(self=@n2, BoneName=clavicle_l, BaseName=spine_05))
   > Set clavicle_up_l(clavicle_up_l:double=MapRangeClamped(Value=BreakRotator(InRot=@n3.Rotation).Pitch, InRangeA=10, InRangeB=40)) -> (Output_Get:double)
   n4 = BreakTransform(InTransform=GetDeltaTransformFromRefPose(self=@n2, BoneName=clavicle_r, BaseName=spine_05))
   > Set clavicle_up_r(clavicle_up_r:double=MapRangeClamped(Value=BreakRotator(InRot=@n4.Rotation).Pitch, InRangeA=10, InRangeB=30)) -> (Output_Get:double)
   n5 = GetOwnerSkeletalMeshComponent()
   n6 = BreakRotator(InRot=BreakTransform(InTransform=GetDeltaTransformFromRefPose(self=@n5, BoneName=upperarm_r, BaseName=spine_05)).Rotation)
   Set ArmUp_R(execute=@n1.then_1, ArmUp_R:double=MapRangeClamped(Value="Absolute (Float)"(A=@n6.Yaw), InRangeA=0, InRangeB=200)) -> (Output_Get:double)
   n7 = GetOwnerSkeletalMeshComponent()
   n8 = BreakRotator(InRot=BreakTransform(InTransform=GetDeltaTransformFromRefPose(self=@n7, BoneName=upperarm_l, BaseName=spine_05)).Rotation)
   > Set ArmUp_L(ArmUp_L:double=MapRangeClamped(Value="Absolute (Float)"(A=@n8.Yaw), InRangeA=40, InRangeB=100)) -> (Output_Get:double)
   > n9 = Branch(Condition=Get bIsTargeting())
   n10 = GetOwnerSkeletalMeshComponent()
   n11 = BreakRotator(InRot=BreakTransform(InTransform=GetDeltaTransformFromRefPose(self=@n10, BoneName=head, BaseName=spine_05)).Rotation)
   > n12 = Set targetingDown_alpha(targetingDown_alpha:double=MapRangeClamped(Value="Absolute (Float)"(A=@n11.Yaw), InRangeA=0, InRangeB=50)) -> (Output_Get:double)
   n13 = Branch(execute=[@n12.then, @n15.then], Condition=Get bIsInFrontEnd())
   > Set normalAlpha L(normalAlpha L:double=0.1) -> (Output_Get:double)
   > Set normalAlpha R(normalAlpha R:double=0.1) -> (Output_Get:double)
   > n14 = Branch(Condition=CheckForBackbling())
   > Set frontendBackbling Alpha(frontendBackbling Alpha:double=1) -> (Output_Get:double)
   Set frontendBackbling Alpha(execute=@n14.else, frontendBackbling Alpha=0)
   Set normalAlpha L(execute=@n13.else, normalAlpha L=0.3)
   > Set normalAlpha R(normalAlpha R=0.9)
   n15 = Set targetingDown_alpha(execute=@n9.else, targetingDown_alpha=0)
   Get bIsTargeting()
  ]}
]
variables [
 ArmUp_L {type=number}
 ArmUp_R {type=number}
 clavicle_up_l {type=number}
 clavicle_up_r {type=number}
 frontendBackbling Alpha {type=number}
 normalAlpha L {type=number}
 normalAlpha R {type=number}
 targetingDown_alpha {type=number}
]
